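; Guided motion config for a constant-force dash movement mode, overriding how the dash is aimed, timed and ended.
; Direction is the normalized forward vector of the owner pawn's camera targeting transform, and height above ground is enforced only when the runtime data's local height is positive.
; Max duration is hotfixable targeting length safely divided by force strength.
; Finish velocity is the local direction scaled by a hotfixable end speed.
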(fn GetDirection_ConstantForce
  (params (GuidedMotionRuntimeData :type Fort%20Movement%20Mode%20Guided%20Motion%20Runtime%20Data%20Object%20Reference))
  (bind (_returnvalue_location _returnvalue_rotation) (Math|Transform|BreakTransform (Ability|GetTargetingTransform (Variables|FortMovementMode_BaseExtLogic|GetOwnerPawn) "Camera" 0)))
  (return (Math|Vector|Normalize (Math|Vector|GetForwardVector _returnvalue_rotation))))

(fn GetEnforceHeightAboveGround_ConstantForce
  (params (GuidedMotionRuntimeData :type Fort%20Movement%20Mode%20Guided%20Motion%20Runtime%20Data%20Object%20Reference))
  (return (> (Variables|/Script/FortniteGame.FortMovementMode_GuidedMotionRuntimeData|GetLocalHeightAboveGround GuidedMotionRuntimeData) 0)))

(fn GetFinishVelocityToSet
  (params (GuidedMotionRuntimeData :type Fort%20Movement%20Mode%20Guided%20Motion%20Runtime%20Data%20Object%20Reference))
  (return (* (Variables|/Script/FortniteGame.FortMovementMode_GuidedMotionRuntimeData|GetLocalDirection GuidedMotionRuntimeData) (ScalableFloat|GetValueatLevel (Variables|Config-Hotfix|GetHotfix_ConstantForceDashEndSpeed)))))

(fn GetMaxDuration
  (params (GuidedMotionRuntimeData :type Fort%20Movement%20Mode%20Guided%20Motion%20Runtime%20Data%20Object%20Reference))
  (return (Math|Float|SafeDivide (ScalableFloat|GetValueasDouble (Variables|Config-Hotfix|GetHotfix_TargetingLength)) (ScalableFloat|GetValueatLevel (Variables|Config-ConstantForce|GetStrength)))))
