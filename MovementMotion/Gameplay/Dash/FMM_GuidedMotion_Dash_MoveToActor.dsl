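; Guided motion movement mode for a dash that moves the player toward a target actor's location.
; Max duration is the distance from the player's initial location to the target's stored initial location, safely divided by a dash speed.
; Finish velocity follows the normalized direction between those two points, scaled by a dash end speed.
; Both speeds come from hotfixable scalable floats.
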(fn GetFinishVelocityToSet
  (params (GuidedMotionRuntimeData :type Fort%20Movement%20Mode%20Guided%20Motion%20Runtime%20Data%20Object%20Reference))
  (return (* (Math|Vector|Normalize (Math|Vector|GetUnitDirection(Vector) (Variables|/Script/FortniteGame.FortMovementMode_GuidedMotionRuntimeData|GetLocalInitialPlayerLocation GuidedMotionRuntimeData) (Variables|/Script/FortniteGame.FortMovementMode_GuidedMotionRuntimeData|GetNetInitialWorldLocation_MoveToActor GuidedMotionRuntimeData))) (ScalableFloat|GetValueatLevel (Variables|Config-Hotfix|GetHotfix_MoveToTargetDashEndSpeed)))))

(fn GetMaxDuration
  (params (GuidedMotionRuntimeData :type Fort%20Movement%20Mode%20Guided%20Motion%20Runtime%20Data%20Object%20Reference))
  (return (Math|Float|SafeDivide (Math|Vector|Distance(Vector) (Variables|/Script/FortniteGame.FortMovementMode_GuidedMotionRuntimeData|GetLocalInitialPlayerLocation GuidedMotionRuntimeData) (Variables|/Script/FortniteGame.FortMovementMode_GuidedMotionRuntimeData|GetNetInitialWorldLocation_MoveToActor GuidedMotionRuntimeData)) (ScalableFloat|GetValueatLevel (Variables|Config-Hotfix|GetHotfix_DashSpeed)))))
