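; Parent gameplay ability for the Athena air dash active power, which ends if a PreLaunch event interrupts it.
; After commit it leaps a tick later, then on launch sets a camera mode, releases balloons and applies a block-until-landed effect.
; Guided-motion MMEs dash toward a target from overridable GetTargetingHit or a fallback sync targeting task, otherwise straight with constant force.
; MME gameplay events track the dash, and ability end cleans up.
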(event Ability|EventActivateAbility
  ;[0.04741 0.276 0.08513] ActivateAbility. End ability if ability costs are not met
  (Ability|ResetTransientState)
  (CallFunction|GetAvatarasValidFortPlayerPawn
    (:ValidFortPlayerPawn
      (if (not (Variables|Settings|GetbSkipCommitAbilityOnActivate))
        (bind _returnvalue (Ability|CommitAbility))
        (if (not _returnvalue)
          (Ability|EndAbility)))
      (|CallParentFunction)
      (Fortnite|FortShippingLog self "ActivateAbility: Commit success")
      ;; End the ability if the player is getting launched - example, from the launch pad which uses the LaunchCharacter() function.
      (bind _asynctaskproxy (Ability|Tasks|WaitGameplayEvent "(TagName=\"Event.Character.PreLaunch\")" 0 true)
        (:then
          (Variables|Transient|AsyncTask|SetAsyncTask_PreLaunchEvent _asynctaskproxy)
          (Utilities|FlowControl|DelayUntilNextTick) ; Activation of this GA is relatively expensive, \nwait a frame before leaping.
          (CallFunction|Leap))
        (:EventReceived
          (Ability|EndAbility) ; This clears all timers and delay.
        )))
    (:AvatarCastFailed
      _)))

(event Custom|Setup%20MME%20Binds
  (bind _returnvalue (Ability|GetAvatarActorfromActorInfo))
  ;[0.09896 0.03317 0.07202] LaunchPlayer: Overridden to run custom launch behavior for this item.
  ;;[0.02439 0.09375 0.05299] Set up binds to react to dash MME events start and end.
  (Utilities|FlowControl|Sequence
    (:then_0
      (bind _asynctaskproxy (Ability|Async|WaitGameplayEventToActor _returnvalue (Variables|Settings|GetMME%20Start%20Tag))
        (:then
          (Variables|Transient|AsyncTask|SetAsyncAction_WaitMMEStart _asynctaskproxy))
        (:EventReceived
          (if (not (Variables|Transient|GetbDashMMEStartTriggered))
            (CallFunction|OnMMEStart)))))
    (:then_1
      (bind _asynctaskproxy_2 (Ability|Async|WaitGameplayEventToActor _returnvalue (Variables|Settings|GetMME%20End%20Tag))
        (:then
          (Variables|Transient|AsyncTask|SetAsyncAction_WaitMMEEnd _asynctaskproxy_2))
        (:EventReceived
          (if (not (Variables|Transient|GetbDashMMEEndTriggered))
            (CallFunction|OnMMEEnd)))))
    (:then_2
      (bind _asynctaskproxy_3 (GameplayTasks|TaskWaitDelay 0 (ScalableFloat|GetValueatLevel (Variables|Hotfix|GetHotfix_MMETimeOutDelay))))
      (Variables|Transient|AsyncTask|SetAsyncTask_MMETimeOutDelay _asynctaskproxy_3))))

(event EventLaunchPlayer
  ;[0.09896 0.03317 0.07202] LaunchPlayer: Overridden to run custom launch behavior for this item.
  (CallFunction|GetAvatarasValidFortPlayerPawn
    (:ValidFortPlayerPawn
      (CallFunction|PreLaunchSetup)
      (CallFunction|FindDashTarget))
    (:AvatarCastFailed
      ;;[1 0 0] Warning: PlayerPawn is null, ending ability.
      (Fortnite|FortShippingLog self "LaunchPlayer: PlayerPawn is null, ending ability." true)
      (Ability|EndAbility))))

(event Custom|On%20MME%20Start
  ;[0.09896 0.03317 0.07202] LaunchPlayer: Overridden to run custom launch behavior for this item.
  ;;[0.04973 0.3198 0.309 0.5] Handle MME start / end
  (Variables|Transient|SetbDashMMEStartTriggered true)
  (Fortnite|FortShippingLog self "LaunchPlayer: Dash event started."))

(event Custom|On%20MME%20End
  ;[0.09896 0.03317 0.07202] LaunchPlayer: Overridden to run custom launch behavior for this item.
  ;;[0.04973 0.3198 0.309 0.5] Handle MME start / end
  (Variables|Transient|SetbDashMMEEndTriggered true)
  (Fortnite|FortShippingLog self "LaunchPlayer: Dash event ended.")
  (CallFunction|EndDash))

(event Custom|On%20MME%20Time%20Out
  ;[0.09896 0.03317 0.07202] LaunchPlayer: Overridden to run custom launch behavior for this item.
  ;;[0.04973 0.3198 0.309 0.5] Handle MME start / end
  (CallFunction|EndDash))

(event Custom|Pre%20Launch%20Setup
  ;[0.09896 0.03317 0.07202] LaunchPlayer: Overridden to run custom launch behavior for this item.
  (Dash|PreparePlayerforLaunch)
  (CallFunction|SetupMMEBinds))

(event Custom|Find%20Dash%20Target
  (bind _targetinghitresult (Targeting|GetTargetingHit))
  (bind (_bblockinghit _binitialoverlap _time _distance _location _impactpoint _normal _impactnormal _physmat _hitactor) (Collision|BreakHitResult _targetinghitresult))
  (bind _targetingpreset (Variables|Settings|GetTargetingPreset))
  ;[0.09896 0.03317 0.07202] LaunchPlayer: Overridden to run custom launch behavior for this item.
  ;[0.05347 0.08302 0.1875] Calculate the dash location by taking into consideration the targeting information held by the targeting component.
  (Utilities|IsServer
    (:true
      (if (CallFunction|IsAI)
        _))
    (:false
      _))
  (if (Collision|BreakHitResult _targetinghitresult)
    (CallFunction|PerformDashtoTarget _hitactor)
    (else
      ;;[0.1837 0.07231 0.1979] If the targeting component has NO valid target, be fair to the player and run a sync trace since the targeting component runs async traces.
      (Utilities|IsValid _targetingpreset
        (:Is%20Valid
          (bind (_asynctaskproxy _targetdata) (Ability|Tasks|PerformTargetingAbilityTask "Task_CalculateDashLoc" _targetingpreset false true)
            (:then
              (Variables|Transient|AsyncTask|SetAsyncTask_CalculateDashLocation _asynctaskproxy))
            (:Targeted
              _)
            (:Cancelled
              _))
          (if (Ability|TargetData|TargetDataHasHitResult _targetdata)
            (CallFunction|PerformDashtoTarget (.HitActor (Collision|BreakHitResult (Ability|TargetData|GetHitResultfromTargetData _targetdata))))
            (else
              _)))
        (:Is%20Not%20Valid
          _))
      (CallFunction|PerformStraightDash))))

(event Custom|Perform%20Dash%20To%20Target
  (params (DashToTarget :type Actor%20Object%20Reference))
  ;[0.09896 0.03317 0.07202] LaunchPlayer: Overridden to run custom launch behavior for this item.
  ;[0.05347 0.08302 0.1875] Calculate the dash location by taking into consideration the targeting information held by the targeting component.
  ;;[0.09187 0.01364 0.09896] Activate dash to the desired target actor.
  (bind _returnvalue (Game|ApplyGuidedMotionMME (CallFunction|GetAvatarasFortPlayerPawn) (Variables|Settings|GetMME_GuidedMotion_Dash_MoveToActor) DashToTarget))
  (Utilities|IsValid _returnvalue
    (:Is%20Valid
      (Variables|Transient|SetMovementModeRuntimeData _returnvalue)
      (Variables|Transient|SetbUsedMoveToActorMME true))
    (:Is%20Not%20Valid
      (Ability|EndAbility)
      ;;;[1 0 0] Warning: MME failed to execute.
      (Fortnite|FortShippingLog self "LaunchPlayer: Move To Target Actor MME failed to execute." true))))

(event Custom|Perform%20Straight%20Dash
  ;[0.09896 0.03317 0.07202] LaunchPlayer: Overridden to run custom launch behavior for this item.
  ;[0.05347 0.08302 0.1875] Calculate the dash location by taking into consideration the targeting information held by the targeting component.
  ;;[0.0219 0.1198 0.1018] If the sync trace also fails to find a target, execute a constant force root motion MME for the dash.
  (bind _returnvalue (Game|ApplyGuidedMotionMME (CallFunction|GetAvatarasFortPlayerPawn) (Variables|Settings|GetMME_GuidedMotion_Dash_ConstantForce)))
  (Utilities|IsValid _returnvalue
    (:Is%20Valid
      (Variables|Transient|SetMovementModeRuntimeData _returnvalue))
    (:Is%20Not%20Valid
      (Ability|EndAbility)
      ;;;[1 0 0] Warning: MME failed to execute.
      (Fortnite|FortShippingLog self "LaunchPlayer: Constant Force MME failed to execute." true))))

(event Custom|EndDash
  ;[0.09896 0.03317 0.07202] LaunchPlayer: Overridden to run custom launch behavior for this item.
  ;[0.05347 0.08302 0.1875] Calculate the dash location by taking into consideration the targeting information held by the targeting component.
  ;[0.03101 0.1119 0.1458] EndDash
  (if (not (Variables|Transient|GetbDashEnded))
    (Variables|Transient|SetbDashEnded true)
    (Camera|RemoveCameraMode)
    (Balloons|ReleaseBalloons)))

(event Ability|EventOnEndAbility
  (params (bWasCancelled :type Boolean))
  ;[0.05347 0.08302 0.1875] Calculate the dash location by taking into consideration the targeting information held by the targeting component.
  ;[0.08333 0.0335 0.007236] OnEndAbility: Clean up.
  (Utilities|FlowControl|Sequence
    (:then_0
      (GameplayTasks|EndTask (Variables|Transient|AsyncTask|GetAsyncTask_CalculateDashLocation))
      (Ability|Async|EndAction (Variables|Transient|AsyncTask|GetAsyncAction_WaitMMEStart))
      (Camera|RemoveCameraMode)
      (CallFunction|EndDash)
      (Ability|ResetTransientState)
      (bind _outfortplayerpawn (CallFunction|GetAvatarasValidFortPlayerPawn)
        (:ValidFortPlayerPawn
          (bind _asfort_movement_comp_character (Utilities|Casting|CastToFortMovementComp_Character (Variables|/Script/Engine.Character|GetCharacterMovement _outfortplayerpawn)))
          (Class|FortMovementCompCharacter|CancelActiveExtensionwithTag _asfort_movement_comp_character (Variables|Settings|GetMovementModeTag)))))
    (:then_1
      (|CallParentFunction bWasCancelled))))

(fn AddCameraMode
  (bind _cameramodifierclass (Variables|Camera|GetCameraModifierClass))
  (bind _cameramode (Variables|Camera|GetCameraMode))
  (bind _cameramodetag (Variables|Camera|GetCameraModeTag))
  ;[0.000237 0.04228 0.05729] AddCameraMode: Add the desired camera mode to the player pawn.
  (bind _outfortplayerpawn (CallFunction|GetAvatarasValidFortPlayerPawn)
    (:ValidFortPlayerPawn
      (Utilities|IsValid (Fortnite|GetFortPlayerControllerfromActor _outfortplayerpawn)
        (:Is%20Valid
          (bind _playercameramanager (Variables|/Script/Engine.PlayerController|GetPlayerCameraManager (Fortnite|GetFortPlayerControllerfromActor _outfortplayerpawn))
            (:then
              (bind _asfort_player_camera_base (Utilities|Casting|CastToFortPlayerCameraBase _playercameramanager))
              (Utilities|FlowControl|Sequence
                (:then_0
                  (if (and (Utilities|IsValidClass _cameramode) (GameplayTags|IsGameplayTagValid _cameramodetag))
                    (Camera|AddCameraModeOverride _asfort_player_camera_base _cameramode _cameramodetag 1 0 0)))
                (:then_1
                  (if (Utilities|IsValidClass _cameramodifierclass)
                    (bind _returnvalue (Game|Player|AddNewCameraModifier _asfort_player_camera_base _cameramodifierclass))
                    (Variables|Camera|SetCameraModifier _returnvalue)))))
            (:else)))))))

(fn GetTargetingHit
  ;[0.5073 0.4355 0.1572 0.5] Can be overriden with result of custom targeting behaviour to dash twoards target
  (return 0))

(fn IsAI
  (return (Pawn|IsBotControlled (CallFunction|GetAvatarasFortPlayerPawn))))

(fn PreparePlayerForLaunch
  ;[0.03457 0.148 0.151] PreparePlayerForLaunch
  (Camera|AddCameraMode)
  (Balloons|ReleaseBalloons)
  ;;[0.1979 1 1] Apply block gameplay effect until the player touches the ground again.
  (Ability|ApplyGameplayEffectToOwner (Variables|Settings|GetGE_BlockTillLand)))

(fn ReleaseBalloons
  (Ability|SendGameplayEventtoActor (Ability|GetAvatarActorfromActorInfo) "(TagName=\"Event.Athena.Balloons.RemoveAll\")"))

(fn RemoveCameraMode
  ;[0.009367 0.09587 0.1042] RemoveCameraMode: Remove the applied camera mode.
  (bind _outfortplayerpawn (CallFunction|GetAvatarasValidFortPlayerPawn)
    (:ValidFortPlayerPawn
      (Utilities|IsValid (Fortnite|GetFortPlayerControllerfromActor _outfortplayerpawn)
        (:Is%20Valid
          (bind _playercameramanager (Variables|/Script/Engine.PlayerController|GetPlayerCameraManager (Fortnite|GetFortPlayerControllerfromActor _outfortplayerpawn))
            (:then
              (bind _asfort_player_camera_base (Utilities|Casting|CastToFortPlayerCameraBase _playercameramanager))
              (Camera|RemoveCameraModeOverride _asfort_player_camera_base (Variables|Camera|GetCameraModeTag) 0 true)
              (bind _cameramodifier (Variables|Camera|GetCameraModifier)
                (:then
                  (Game|Player|RemoveCameraModifier _asfort_player_camera_base _cameramodifier))
                (:else)))
            (:else)))))))

(fn ResetTransientState
  ;[0.08847 0.15 0.03889 0.5] Crucial functionality to keep up to date. Resets all the transient variables to their default values so that the GA can be executed correctly when invoked multiple times per instance.
  (Variables|Transient|SetbDashEnded false)
  (Variables|Transient|SetbUsedMoveToActorMME false)
  (Variables|Transient|SetbDashMMEStartTriggered false)
  (Variables|Transient|SetbDashMMEEndTriggered false))
